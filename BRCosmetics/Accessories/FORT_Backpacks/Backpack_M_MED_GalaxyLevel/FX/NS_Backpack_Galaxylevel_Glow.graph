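# Niagara system for a pink glow flare on the Galaxy Level backpack cosmetic.
# A single Flare emitter burst-spawns persistent looping sprites that follow a sampled, offset location on the user-supplied skeletal mesh, with fixed local facing and alignment.
# Brightness scales by time of day, dimmer at night.
# Two sprite renderers are present, and extra parameters derive a dimmer, half-width side glow from the particle's color and size.
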NiagaraSystem /BRCosmetics/Accessories/FORT_Backpacks/Backpack_M_MED_GalaxyLevel/FX/NS_Backpack_Galaxylevel_Glow
dependencies {usedDataInterfaces [], usedDynamicInputs [], usedModules [&SystemState]}
emitters {Flare {
  emitterUpdateScript [
   &EmitterState(Life Cycle Mode=::NewEnumerator1, Loop Behavior=::NewEnumerator1, Loop Duration Mode=::NewEnumerator1, UseLoopDelay=false)
   &SpawnBurst_Instantaneous(Use Spawn Probability=false, UseLoopCountLimit=false)
  ]
  particleSpawnScript [
   &InitializeParticle(
    Color=<rgba 2 0.2292 1.6012 1>
    Lifetime=5
    Sprite Size=<26 13>
    Sprite Size Mode=::NewEnumerator4
    UsePositionOffset=false
    Write Material Random=true
   )
   &CameraOffset(Camera Offset Amount=1.5)
  ]
  particleUpdateScript [
   &ParticleState(
    Kill Particles When Lifetime Has Elapsed=false
    Let Infinitely Lived Particles Die When Emitter Deactivates=true
    Loop Particles Lifetime=true
   )
   &SkeletalMeshLocation(Bone Sampling Mode=::NewEnumerator4, MaskSpawnGroup=false, Sampled Position Offset=<0 -14 5.7>, Skeletal Mesh=$User.SkeletalMesh)
   &ScaleColor(
    Scale RGB=&Multiply_VectorByFloat(Vector=<1 1 1>, Float=&FN_Select_Float_from_Time_Of_Day(Morning=2, Day=2, Evening=2, Night=0.5))
    ScaleA=true
    ScaleRGB=true
   )
   &SpriteFacingAndAlignment(
    Alignment Coordinate Space=::Local
    Facing Coordinate Space=::Local
    Sprite Alignment=<0 0 1>
    Sprite Facing=<0 -1 0>
    WriteAlignment=true
    WriteFacing=true
   )
   /SetVariables(
    Particles.CameraDistanceSideGlow=1.5
    Particles.EdgeGlowStrenght=0.5
    Particles.SideGlowColor:LinearColor=&MultiplyLinearColor(Linear Color=$Particles.Color, Scale Factor=&MakeVector4_FromFloat(Float=0.3))
    Particles.SideGlowSize:Vector2f=&Multiply_Vector2D(A=$Particles.SpriteSize, B=&Multiply_Vector2D_ByFloat(Float=1, Vector2D=<0.5 1>))
   ) {isSetParameters=true}
  ]
  rendererClasses [&NiagaraSpriteRendererProperties]
  renderers [
   {rendererClass=&NiagaraSpriteRendererProperties}
   {rendererClass=&NiagaraSpriteRendererProperties, rendererIndex=1}
  ]
 }}
summary {
 systemName=NS_Backpack_Galaxylevel_Glow
 emitters [{emitterName=Flare, rendererClasses [&NiagaraSpriteRendererProperties]}]
 userVariables [User.SkeletalMesh {
   type=NiagaraDataInterfaceSkeletalMesh
   defaultValue {
    struct=NiagaraExt_VariableValue_DataInterface
    value {dataInterface=&self:NiagaraDataInterfaceSkeletalMesh_0}
   }
  }]
}
