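# Animation Blueprint for the GalaxyLevel backpack that toggles a material element to match the pickaxe and skydiving state.
# Weapon-change and customization-loaded events fetch the mesh's dynamic material instance, then check for an equipped pickaxe tagged GalaxyLevel.
# "Hide Element 01" is 1 when that pickaxe is equipped outside skydiving, otherwise 0; skydive begin and end events also update it.
# The AnimGraph only passes the input pose through.
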AnimBlueprint /BRCosmetics/Accessories/FORT_Backpacks/Backpack_M_MED_GalaxyLevel/Meshes/M_MED_GalaxyLevel_Pack_AnimBP
target_skeleton=&/M_MED_GalaxyLevel_Pack_Skeleton
graphs [
 AnimGraph {nodes [Output Pose(Result=Input Pose()) {layerGroup=None}]}
 EventGraph {nodes [
   n1 = Event OnWeaponChanged()
   n2 = Get Dynamic Material Instance(execute=[@n1.then, @n8.then], SkeletalMeshComponent=GetPartSkeletalMeshComponent())
   n3 = PickaxeItemDefinitionHasAnyMetaTag(Tags=Make Array("[0]"=(TagName="Cosmetics.MetaTags.GalaxyLevel")), bRequiresEquipped=true)
   > n4 = Branch(Condition=@n3)
   > n5 = Branch(Condition=Get bIsSkyDiving())
   SetScalarParameterValue(execute=[@n6.then, @n5.then, @n7.else, @n4.else], self=@n2, ParameterName=Hide Element 01, Value=0)
   SetScalarParameterValue(execute=[@n5.else, @n7.then], self=@n2, ParameterName=Hide Element 01, Value=1)
   n6 = Event OnBeginSkydiving()
   Event OnEndSkydiving()
   > n7 = Branch(Condition=@n3)
   n8 = Event OnCharacterCustomizationLoaded()
   TryGetPawnOwner()
   Event BlueprintUpdateAnimation()
  ]}
]
variables []
